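# Animation Blueprint for the Voyager Remix skull face accessory, the NoFloat variant, with an optional bobbing offset on the head bone.
# The AnimGraph copies the attached parent mesh's pose, then additively translates dyn_head in bone space by ShardFloatA.
# The update event builds that vector from game-time sine waves at differing per-axis rates and amplitudes.
# It multiplies the result by integer FloatingHeadOn, so zero disables the float.
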AnimBlueprint /BRCosmetics/Characters/Player/Male/Medium/Bodies/M_MED_VoyagerRemix/Meshes/Parts/Skull/M_MED_VoyagerRemix_Skull_FaceAcc_AnimBP_NoFloat
target_skeleton=&/M_MED_Voyager_Remix_Skull_Flames_FaceAcc_Skeleton
graphs [
 AnimGraph {nodes [
   n1 = "Transform (Modify) Bone - Bone: dyn_head"(
    ComponentPose=Local To Component(LocalPose=Copy Pose From Mesh() {bUseAttachedParent=true})
    Alpha=1
    Translation=Get ShardFloatA() -> (ShardFloatA:Vector)
    Rotation="0,0,0"
    Scale="1,1,1"
   ) {
    rotationMode=BMM_Additive
    rotationSpace=BCS_BoneSpace
    translationMode=BMM_Additive
    translationSpace=BCS_BoneSpace
    alphaBoolBlend {blendOption=Linear}
    boneToModify {boneName=dyn_head}
   }
   Output Pose(Result=Component To Local(ComponentPose=@n1)) {layerGroup=None}
  ]}
 EventGraph {nodes [
   Event BlueprintUpdateAnimation()
   n1 = GetGameTimeInSeconds()
   n2 = float * float(A=1, B=1.1)
   n3 = float * float(A="Sin (Radians)"(A=float * float(A=@n2, B=@n1)), B=0.5)
   n4 = float * float(A=1, B=1.3)
   n5 = float * float(A="Sin (Radians)"(A=float * float(A=@n4, B=@n1)), B=0.6)
   n6 = float * float(A=1, B=1.5)
   n7 = float * float(A="Sin (Radians)"(A=float * float(A=@n6, B=@n1)), B=0.7)
   > Set ShardFloatA(ShardFloatA:Vector=vector * integer!(A:Vector=MakeVector(X=@n7, Y=@n3, Z=@n5), B:int=Get FloatingHeadOn() -> (FloatingHeadOn:int=0)) -> (ReturnValue:Vector)) -> (Output_Get:Vector)
   TryGetPawnOwner()
   GetGameTimeInSeconds()
  ]}
]
variables [
 FloatingHeadOn {type=integer}
 ShardFloatA {type=Vector}
]
